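# Physics asset for the ShimPilon backpack cosmetic, covering two kunai and a dangling belt.
# A box body on pack_belt_base anchors a six-segment capsule chain, whose radii mostly widen toward the tip.
# Each belt link is constrained to the previous one with small swing limits and zero twist limit.
# The left and right kunai each have a capsule body with no constraints.
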PhysicsAsset /BRCosmetics/Accessories/FORT_Backpacks/Backpack_M_MED_ShimPilon/Meshes/M_MED_ShimPilon_Pack_Physics
bodies {
 Kunai_L [{
   shapeType=Capsule
   length=16.8999
   radius=3.1775
   shapeName=Kunai_L_capsule
   center=<-5.2159 0.0121 -0.3075>
   rotation=<rot 0 89.8461 -89.6803>
  }]
 Kunai_R [{
   shapeType=Capsule
   length=16.8999
   radius=3.1775
   shapeName=Kunai_R_capsule
   center=<4.9303 -0.0114 0.3059>
   rotation=<rot 0 89.8461 -89.6803>
  }]
 pack_belt_1 [{shapeType=Capsule, length=2.3878, radius=1.6, shapeName=pack_belt_1_capsule, center=<2.3203 -1.25 0>, rotation=<rot 90 5.6252 0>}]
 pack_belt_2 [{shapeType=Capsule, length=6.8, radius=1.7, shapeName=pack_belt_2_capsule, center=<5.2157 -1.25 0>, rotation=<rot 90 0 0>}]
 pack_belt_3 [{shapeType=Capsule, length=6, radius=4, shapeName=pack_belt_3_capsule, center=<5 -3.3184 0>, rotation=<rot 90 0 0>}]
 pack_belt_4 [{shapeType=Capsule, length=8, radius=5, shapeName=pack_belt_4_capsule, center=<5 0 0>, rotation=<rot 90 0 0>}]
 pack_belt_5 [{shapeType=Capsule, length=5, radius=5, shapeName=pack_belt_5_capsule, center=<5 0 0>, rotation=<rot 90 0 0>}]
 pack_belt_6 [{shapeType=Capsule, length=2, radius=6, shapeName=pack_belt_6_capsule, center=<2.4151 0 0>, rotation=<rot 90 0 0>}]
 pack_belt_base [{shapeType=Box, extentX=12, extentY=23, extentZ=16, shapeName=pack_belt_base_box, center=<2.7316 6.562 0>, rotation=<rot 0 -84.374 0>}]
}
constraints [
 {bone1Name=pack_belt_1, bone2Name=pack_belt_base, swing1LimitDegrees=15, swing2LimitDegrees=10, twistLimitDegrees=0}
 {bone1Name=pack_belt_2, bone2Name=pack_belt_1, swing1LimitDegrees=20, swing2LimitDegrees=15, twistLimitDegrees=0}
 {bone1Name=pack_belt_3, bone2Name=pack_belt_2, swing1LimitDegrees=10, swing2LimitDegrees=5, twistLimitDegrees=0}
 {bone1Name=pack_belt_4, bone2Name=pack_belt_3, swing1LimitDegrees=10, swing2LimitDegrees=5, twistLimitDegrees=0}
 {bone1Name=pack_belt_5, bone2Name=pack_belt_4, swing1LimitDegrees=10, swing2LimitDegrees=5, twistLimitDegrees=0}
 {bone1Name=pack_belt_6, bone2Name=pack_belt_5, swing1LimitDegrees=10, swing2LimitDegrees=5, twistLimitDegrees=0}
]
