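# Animation Blueprint for the ShimPilon backpack cosmetic, adding secondary physics to the pack, its two kunai and its belt.
# bIsInFrontEnd selects between two RigidBody plus kunai AnimDynamics branches with different simulation tuning and spaces.
# The result runs through per-bone AnimDynamics on six belt bones, then a Trail controller.
# The event graph updates gravity for one RigidBody branch, with a sliding override, plus LOD, RBAN-disabled and velocity-opposed force variables.
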AnimBlueprint /BRCosmetics/Accessories/FORT_Backpacks/Backpack_M_MED_ShimPilon/Meshes/M_MED_ShimPilon_Pack_AnimBP
target_skeleton=&/M_MED_ShimPilon_Pack_Skeleton
graphs [
 AnimGraph {nodes [
   n1 = RigidBody!(ComponentPose=Local To Component(LocalPose=Use cached pose 'input'()), Alpha=0.7, OverrideWorldGravity:Vector="0,0,-980", ExternalForce) -> (Pose) {
    bClampLinearTranslationLimitToRefPose=true
    bOverrideWorldGravity=true
    bUseExternalClothCollision=true
    overridePhysicsAsset=&/M_MED_ShimPilon_Pack_Physics
    simulationSpace=BaseBoneSpace
    alphaBoolBlend {blendInTime=1, blendOption=Linear, blendOutTime=2}
    baseBoneRef {boneName=root}
    componentAppliedLinearAccClamp=<0 0 0>
    overrideWorldGravity {z=-980}
    simSpaceSettings {dampingAlpha=0, maxAngularAcceleration=25, maxAngularVelocity=5, maxLinearAcceleration=400, maxLinearVelocity=100, worldAlpha=1}
   }
   n2 = Anim Dynamics - Bone: Kunai_R(ComponentPose=@n1, Alpha=1, ExternalForce="0,0,0") {
    angularDampingOverride=0.9999
    bOverrideAngularDamping=true
    bUseGravityOverride=true
    simulationSpace=World
    boundBone {boneName=Kunai_R}
    componentAppliedLinearAccClamp=<5000 5000 5000>
    componentLinearAccScale=<0.5 0.5 0.5>
    componentLinearVelScale=<0.5 0.5 0.5>
    gravityOverride {z=-980}
    physicsBodyDefinitions [{
      boundBone {boneName=Kunai_R}
      boxExtents=<10 10 10>
      constraintSetup {angularLimitsMax=<3 5 5>, angularLimitsMin=<-3 -5 -5>, angularTarget=<0 0 0>}
      localJointOffset=<0.1 0 0>
     }]
    simSpaceSettings {maxAngularAcceleration=35, maxAngularVelocity=7, simSpaceAngularAlpha=1}
   }
   n3 = Anim Dynamics - Bone: Kunai_L(ComponentPose=@n2, Alpha=1, ExternalForce="0,0,0") {
    angularDampingOverride=0.9999
    bOverrideAngularDamping=true
    bUseGravityOverride=true
    simulationSpace=World
    boundBone {boneName=Kunai_L}
    componentAppliedLinearAccClamp=<5000 5000 5000>
    componentLinearAccScale=<0.5 0.5 0.5>
    componentLinearVelScale=<0.5 0.5 0.5>
    gravityOverride {z=-980}
    physicsBodyDefinitions [{
      boundBone {boneName=Kunai_L}
      boxExtents=<10 10 10>
      constraintSetup {angularLimitsMax=<3 5 5>, angularLimitsMin=<-3 -5 -5>, angularTarget=<0 0 0>}
      localJointOffset=<-0.1 0 0>
     }]
    simSpaceSettings {maxAngularAcceleration=35, maxAngularVelocity=7, simSpaceAngularAlpha=1}
   }
   n4 = Component To Local(ComponentPose=@n3)
   n5 = RigidBody(ComponentPose=Local To Component(LocalPose=Use cached pose 'input'()), OverrideWorldGravity=Get GravityOverride()) {
    bClampLinearTranslationLimitToRefPose=true
    bOverrideWorldGravity=true
    bUseExternalClothCollision=true
    overridePhysicsAsset=&M_MED_ShimPilon_Pack_Physics
    simulationSpace=BaseBoneSpace
    alphaBoolBlend {blendInTime=1, blendOption=Linear, blendOutTime=2}
    baseBoneRef {boneName=root}
    componentAppliedLinearAccClamp=<0 0 0>
    overrideWorldGravity {z=-980}
    simSpaceSettings {
     dampingAlpha=0
     maxAngularAcceleration=30
     maxAngularVelocity=6
     maxLinearAcceleration=900
     maxLinearVelocity=180
     worldAlpha=1
     externalLinearDragV=<0.7 0.7 0.7>
    }
   }
   n6 = Anim Dynamics - Bone: Kunai_R(ComponentPose=@n5, Alpha=1, ExternalForce="0,0,0") {
    angularDampingOverride=0.9999
    bOverrideAngularDamping=true
    bUseGravityOverride=true
    simulationSpace=RootRelative
    boundBone {boneName=Kunai_R}
    componentAppliedLinearAccClamp=<5000 5000 5000>
    componentLinearAccScale=<0.5 0.5 0.5>
    componentLinearVelScale=<0.5 0.5 0.5>
    gravityOverride {z=-980}
    physicsBodyDefinitions [{
      boundBone {boneName=Kunai_R}
      boxExtents=<10 10 10>
      constraintSetup {angularLimitsMax=<3 5 5>, angularLimitsMin=<-3 -5 -5>, angularTarget=<0 0 0>}
      localJointOffset=<0.1 0 0>
     }]
    simSpaceSettings {maxAngularAcceleration=35, maxAngularVelocity=7, simSpaceAngularAlpha=1}
   }
   n7 = Anim Dynamics - Bone: Kunai_L(ComponentPose=@n6, Alpha=1, ExternalForce="0,0,0") {
    angularDampingOverride=0.9999
    bOverrideAngularDamping=true
    bUseGravityOverride=true
    simulationSpace=RootRelative
    boundBone {boneName=Kunai_L}
    componentAppliedLinearAccClamp=<5000 5000 5000>
    componentLinearAccScale=<0.5 0.5 0.5>
    componentLinearVelScale=<0.5 0.5 0.5>
    gravityOverride {z=-980}
    physicsBodyDefinitions [{
      boundBone {boneName=Kunai_L}
      boxExtents=<10 10 10>
      constraintSetup {angularLimitsMax=<3 5 5>, angularLimitsMin=<-3 -5 -5>, angularTarget=<0 0 0>}
      localJointOffset=<-0.1 0 0>
     }]
    simSpaceSettings {maxAngularAcceleration=35, maxAngularVelocity=7, simSpaceAngularAlpha=1}
   }
   n8 = Component To Local(ComponentPose=@n7)
   n9 = Get bIsInFrontEnd()
   n10 = Anim Dynamics - Bone: pack_belt_1!(
    ComponentPose=Local To Component(LocalPose=Blend Poses by bool(BlendPose_0=@n4, BlendPose_1=@n8, bActiveValue=@n9))
    bAlphaBoolEnabled:bool=True
    GravityOverride:Vector="0,0,-980"
    ExternalForce="0,0,0"
   ) -> (Pose) {
    alphaInputType=Bool
    angularDampingOverride=0.9999
    angularSpringConstant=100
    bAngularSpring=true
    bOverrideAngularDamping=true
    bUseGravityOverride=true
    simulationSpace=RootRelative
    boundBone {boneName=pack_belt_1}
    componentAppliedLinearAccClamp=<2500 2500 2500>
    componentLinearAccScale=<0.2 0.2 0.2>
    componentLinearVelScale=<0.3 0.2 0.2>
    physicsBodyDefinitions [{
      boundBone {boneName=pack_belt_1}
      boxExtents=<9 9 9>
      constraintSetup {angularLimitsMax=<10 10 0>, angularLimitsMin=<-10 -10 -10>, angularTarget=<1 0 0>}
      localJointOffset=<1 0 0>
     }]
   }
   n11 = Anim Dynamics - Bone: pack_belt_2!(ComponentPose=@n10, bAlphaBoolEnabled:bool=True, ExternalForce="0,0,0") -> (Pose) {
    alphaInputType=Bool
    angularDampingOverride=0.99
    bOverrideAngularDamping=true
    simulationSpace=RootRelative
    boundBone {boneName=pack_belt_2}
    componentAppliedLinearAccClamp=<2500 2500 2500>
    componentLinearAccScale=<0.2 0.2 0.2>
    componentLinearVelScale=<0.3 0.2 0.2>
    physicsBodyDefinitions [{
      boundBone {boneName=pack_belt_2}
      boxExtents=<9 9 9>
      constraintSetup {angularLimitsMax=<15 15 0>, angularLimitsMin=<-15 -15 -15>, angularTarget=<1 0 0>}
      localJointOffset=<1 0 0>
     }]
   }
   n12 = Anim Dynamics - Bone: pack_belt_3!(ComponentPose=@n11, bAlphaBoolEnabled:bool=True, GravityOverride:Vector="0,0,-980", ExternalForce="0,0,0") -> (Pose) {
    alphaInputType=Bool
    angularDampingOverride=0.9999
    angularSpringConstant=100
    bAngularSpring=true
    bOverrideAngularDamping=true
    bUseGravityOverride=true
    simulationSpace=RootRelative
    boundBone {boneName=pack_belt_3}
    componentAppliedLinearAccClamp=<2500 2500 2500>
    componentLinearAccScale=<0.2 0.2 0.2>
    componentLinearVelScale=<0.3 0.2 0.2>
    physicsBodyDefinitions [{
      boundBone {boneName=pack_belt_3}
      boxExtents=<9 9 9>
      constraintSetup {angularLimitsMax=<15 20 20>, angularLimitsMin=<-15 -20 -30>, angularTarget=<1 0 0>}
      localJointOffset=<1 0 0>
     }]
   }
   n13 = Anim Dynamics - Bone: pack_belt_4!(ComponentPose=@n12, bAlphaBoolEnabled:bool=True, GravityOverride:Vector="0,0,-980", ExternalForce="0,0,0") -> (Pose) {
    alphaInputType=Bool
    angularDampingOverride=0.9999
    angularSpringConstant=100
    bAngularSpring=true
    bOverrideAngularDamping=true
    bUseGravityOverride=true
    simulationSpace=RootRelative
    boundBone {boneName=pack_belt_4}
    componentAppliedLinearAccClamp=<2500 2500 2500>
    componentLinearAccScale=<0.2 0.2 0.2>
    componentLinearVelScale=<0.3 0.2 0.2>
    physicsBodyDefinitions [{
      boundBone {boneName=pack_belt_4}
      boxExtents=<9 9 9>
      constraintSetup {angularLimitsMax=<15 20 20>, angularLimitsMin=<-15 -20 -30>, angularTarget=<1 0 0>}
      localJointOffset=<1 0 0>
     }]
   }
   n14 = Anim Dynamics - Bone: pack_belt_5!(ComponentPose=@n13, bAlphaBoolEnabled:bool=True, GravityOverride:Vector="0,0,-980", ExternalForce="0,0,0") -> (Pose) {
    alphaInputType=Bool
    angularDampingOverride=0.9999
    angularSpringConstant=100
    bAngularSpring=true
    bOverrideAngularDamping=true
    bUseGravityOverride=true
    simulationSpace=RootRelative
    boundBone {boneName=pack_belt_5}
    componentAppliedLinearAccClamp=<2500 2500 2500>
    componentLinearAccScale=<0.2 0.2 0.2>
    componentLinearVelScale=<0.3 0.2 0.2>
    physicsBodyDefinitions [{
      boundBone {boneName=pack_belt_5}
      boxExtents=<9 9 9>
      constraintSetup {angularLimitsMax=<15 20 20>, angularLimitsMin=<-15 -20 -30>, angularTarget=<1 0 0>}
      localJointOffset=<1 0 0>
     }]
   }
   n15 = Anim Dynamics - Bone: pack_belt_6!(ComponentPose=@n14, bAlphaBoolEnabled:bool=True, GravityOverride:Vector="0,0,-980", ExternalForce="0,0,0") -> (Pose) {
    alphaInputType=Bool
    angularDampingOverride=0.9999
    angularSpringConstant=100
    bAngularSpring=true
    bOverrideAngularDamping=true
    bUseGravityOverride=true
    simulationSpace=RootRelative
    boundBone {boneName=pack_belt_6}
    componentAppliedLinearAccClamp=<2500 2500 2500>
    componentLinearAccScale=<0.2 0.2 0.2>
    componentLinearVelScale=<0.3 0.2 0.2>
    physicsBodyDefinitions [{
      boundBone {boneName=pack_belt_6}
      boxExtents=<9 9 9>
      constraintSetup {angularLimitsMax=<20 20 20>, angularLimitsMin=<-20 -20 -30>, angularTarget=<1 0 0>}
      localJointOffset=<1 0 0>
     }]
   }
   n16 = Trail controller - Bone: pack_belt_6!(ComponentPose=@n15, bAlphaBoolEnabled:bool=True, RelaxationSpeedScale:float=1, FakeVelocity:Vector="0,0,0") -> (Pose) {
    alphaInputType=Bool
    bActorSpaceFakeVel=true
    bShowTrailLocation=true
    bUsePlanarLimit=true
    chainLength=6
    debugLifeTime=0.1
    maxDeltaTime=0.0333
    relaxationSpeedScale=1.2
    alphaScaleBiasClamp {bInterpResult=true, interpSpeedDecreasing=25, interpSpeedIncreasing=25}
    baseJoint {boneName=pack}
    rotationLimits [
     {limitMax=<180 180 180>}
     {limitMax=<180 180 180>}
     {limitMax=<180 180 180>}
     {limitMax=<180 180 180>}
     {limitMax=<180 180 180>}
     {limitMax=<180 180 180>}
    ]
    rotationOffsets [<0 0 0>, <0 0 0>, <0 0 0>, <0 0 0>, <0 0 0>, <0 0 0>]
    trailBone {boneName=pack_belt_6}
    trailRelaxationSpeed {editorCurveData {keys [<at 0 15 RCIM_Linear>,<at 0.5289 13.1838 RCIM_Linear>,<at 1 6 RCIM_Linear>]}}
   }
   Output Pose(Result=Component To Local(ComponentPose=@n16))
   Save cached pose 'input'(Pose=Input Pose()) {cacheName=input}
   Get GravityOverride()
  ]}
 EventGraph {nodes [
   Event BlueprintUpdateAnimation()
   > n1 = Sequence!(execute) -> (then_0, then_1, then_2:Exec)
   > n2 = "Gravity Override State Function (Parameter Asset) (Nativized)"(
    CharacterPartAnimInstance=Self-Reference()
    GameTimeInSeconds=GetGameTimeInSeconds()
    Parameters=&/M_Med_ShimPilon_Pack_Parameters
   )
   > Set RBAN Settings Out(RBAN Settings Out=@n2.OutRigidBodyAnimNodeParameters)
   > n3 = Branch(Condition=Get bIsSliding())
   > Set GravityOverride(GravityOverride="0,0,-500")
   Set GravityOverride(execute=@n3.else, GravityOverride=@n2)
   Set Is LOD2+(execute=@n1.then_1, Is LOD2+:bool="integer >= integer"(A=Get CurrentLOD(), B=2)) -> (Output_Get:bool)
   > Set Is RBAN Disabled(Is RBAN Disabled:bool=NOT Boolean(A=Get bRigidBodyEnabled())) -> (Output_Get:bool)
   n4 = Get!() -> (PawnVelocity_X:double, PawnVelocity_Y:double, PawnVelocity_Z:double)
   Set!(
    execute=@n1.then_2
    External Force_X:double=float * float(A=@n4, B=-1) {is=title K2Node_PromotableOperator}
    External Force_Y:double=float * float(A=@n4.PawnVelocity_Y, B=-1)
    External Force_Z:double=0
   ) -> (then, Output_Get:Vector)
  ]}
]
variables [
 Backpack Rotate {type=Rotator}
 Backpack Translation {type=Vector}
 External Force {type=Vector}
 GravityOverride {type=Vector}
 Is LOD2+ {type=boolean}
 Is RBAN Disabled {type=boolean}
 RBAN Settings Out {type=FortRigidBodyAnimNodeParameters}
]
